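# Material function that computes a waving displacement vector for a smoke trail, driven by noise and wind.
# The smoke texture, sampled at UVs shifted in U by Offset, is centered, weighted per axis and scaled by Tangent Multiplier.
# It adds a Direction parameter and local-space Wind weighted by squared Linear Mask.
# Wiggle Scale and Offset plus Non Linear Mask scale the result; a UV1-driven lerp is left unconnected.
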MaterialFunction /Game/Environments/TechArt/SmokeTrail/MF_SmokeTrail_Waving
graph {nodes [
  n1 = FunctionInput() {inputName=Offset, inputType=FunctionInput_Scalar, sortPriority=1}
  n2 = Add(A=FunctionInput() {inputName=UVs, inputType=FunctionInput_Vector2}, B=AppendVector(A=@n1, B=Constant())) {constB=-0.07}
  n3 = Multiply(
   A=Add(A=TextureSample(UVs=@n2) {samplerType=SAMPLERTYPE_LinearColor, texture=&/T_SmokeTrail_D}.RGB) {constB=-0.5}
   B=Constant3Vector() {constant=<rgba 1 0.7 0.3 0>}
  )
  n4 = FunctionInput() {inputName=Tangent Multiplier, inputType=FunctionInput_Scalar, sortPriority=4}
  n5 = FunctionInput() {inputName=Linear Mask, inputType=FunctionInput_Scalar, sortPriority=2}
  n6 = VectorParameter() {group=Waving, parameterName=Wind, sortPriority=7, defaultValue=<rgba -10 0 1 1>}
  n7 = Add(
   A=Add(A=Multiply(A=@n3, B=@n4), B=VectorParameter() {parameterName=Direction, defaultValue=<rgba 0.25 0 3 1>}.RGB)
   B=Multiply(A=Multiply(A=@n5, B=@n5), B=Transform(None=@n6.RGB) {transformSourceType=TRANSFORMSOURCE_World, transformType=TRANSFORM_Local})
  )
  n8 = ScalarParameter() {defaultValue=513.0842, group=Waving, parameterName=Wiggle Scale, sortPriority=5}
  n9 = FunctionInput() {inputName=Non Linear Mask, inputType=FunctionInput_Scalar, sortPriority=3}
  FunctionOutput(None=Multiply(A=Multiply(A=@n7, B=@n8) {constB=2000}, B=Add(A=@n1, B=@n9) {constB=-0.02}) {constB=11})
  LinearInterpolate(Alpha=ComponentMask(None=TextureCoordinate() {coordinateIndex=1}) {g=true}) {constA=0.3}
 ]}
