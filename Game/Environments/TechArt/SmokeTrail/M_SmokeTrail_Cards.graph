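# Unlit smoke trail material that animates mesh cards in the vertex stage, computing position offsets to form a wavy, view-oriented trail.
# UV1 plus time gives looping linear and power-curved masks that interpolate card width and size from start to end.
# MF_SmokeTrail_Waving gives the wave offset, with two extra offset samples giving a tangent.
# Cross products with a camera-relative vector build the card shape, from which interpolated normals are derived.
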Material /Game/Environments/TechArt/SmokeTrail/M_SmokeTrail_Cards
graph {
 properties {bAutomaticallySetUsageInEditor=false, shadingModel=MSM_Unlit, usedShadingModels=Unlit}
 nodes [
  n1 = DotProduct(A=Constant3Vector() {constant=<rgba 0 0 1 0>}, B=NamedRerouteUsage())
  Frac(None=Max(A=@n1) {constB=0})
  n2 = Add(A=NamedRerouteUsage(), B=Multiply(A=LocalPosition().XYZ) {constB=-1})
  n3 = Transform(None=WorldPosition() {worldPositionShaderOffset=WPT_CameraRelative}.XYZ) {transformSourceType=TRANSFORMSOURCE_World, transformType=TRANSFORM_Local}
  NamedRerouteDeclaration(None=Normalize(VectorInput=Add(A=@n2, B=@n3))) {name=Card Vector, nodeColor=<rgba 1 0.6797 0 1>}
  n4 = NamedRerouteUsage()
  n5 = Add(A=AppendVector(A=Multiply(A=@n4) {constB=0.6}, B=Constant()), B=Multiply(A=Time(), B=Constant2Vector() {g=0.03, r=-0.1}) {constB=-0.15})
  n6 = MaterialFunctionCall!(
   UVs=@n5
   Offset=Constant() {r=0.02}
   Linear Mask=NamedRerouteUsage()
   Non Linear Mask=NamedRerouteUsage()
   Tangent Multiplier=Constant() {r=0.5}
  ) {materialFunction=&/MF_SmokeTrail_Waving}
  n7 = MaterialFunctionCall(
   UVs=@n5
   Offset=Constant() {r=-0.02}
   Linear Mask=NamedRerouteUsage()
   Non Linear Mask=NamedRerouteUsage()
   Tangent Multiplier=Constant() {r=0.5}
  ) {materialFunction=&MF_SmokeTrail_Waving}
  NamedRerouteDeclaration(None=Normalize(VectorInput=Subtract(A=@n6.Result, B=@n7.Result))) {name=Waving Tangent, nodeColor=<rgba 0 1 0.8594 1>}
  n8 = TextureCoordinate() {coordinateIndex=1}
  n9 = Time()
  n10 = Frac(None=Add(A=ComponentMask(None=@n8) {r=true}, B=Multiply(A=@n9) {constB=0.1}))
  NamedRerouteDeclaration(None=Power(
    Base=NamedRerouteDeclaration(None=@n10) {name=Linear Mask, nodeColor=<rgba 0.9453 0 1 1>}
    Exp=ScalarParameter() {defaultValue=1.6, parameterName=Shape Exponent}
   ) {constExponent=1.6}) {name=Non Linear Mask, nodeColor=<rgba 0.1016 1 0 1>}
  NamedRerouteDeclaration(None=LinearInterpolate(
    A=ScalarParameter() {defaultValue=0.5, parameterName=Width Start}
    B=ScalarParameter() {defaultValue=1, parameterName=Width End}
    Alpha=NamedRerouteUsage()
   )) {name=Card Width, nodeColor=<rgba 1 0 0.4219 1>}
  NamedRerouteDeclaration(None=LinearInterpolate(
    A=ScalarParameter() {defaultValue=0.4, parameterName=Size Start}
    B=ScalarParameter() {defaultValue=2, parameterName=Size End}
    Alpha=NamedRerouteUsage()
   )) {name=Card Size, nodeColor=<rgba 0 0.7422 1 1>}
  n11 = NamedRerouteUsage()
  n12 = CrossProduct(A=@n11, B=NamedRerouteUsage())
  n13 = Normalize(VectorInput=@n12)
  n14 = NamedRerouteUsage()
  n15 = Multiply(A=@n13, B=@n14) {constB=0.25}
  n16 = LocalPosition()
  n17 = ComponentMask(None=@n16.XYZ) {r=true}
  n18 = Multiply(A=Multiply(A=NamedRerouteUsage(), B=VertexColor().R)) {constB=-100}
  n19 = NamedRerouteUsage()
  n20 = CrossProduct(A=@n12, B=@n11)
  n21 = Multiply(A=OneMinus(None=@n14), B=@n20)
  n22 = Multiply(A=CrossProduct(A=@n13, B=@n11), B=@n14) {constB=0.25}
  n23 = ComponentMask(None=@n16.XYZ) {g=true}
  n24 = Add(A=Add(A=Multiply(A=@n15, B=@n17), B=Multiply(A=@n18, B=@n19)), B=Multiply(A=Add(A=@n21, B=@n22), B=@n23))
  n25 = Multiply(A=@n24, B=NamedRerouteUsage())
  NamedRerouteDeclaration(None=Transform(None=Add(A=@n2, B=@n25)) {transformSourceType=TRANSFORMSOURCE_Local}) {name=WPO, nodeColor=<rgba 0.4297 0 1 1>}
  NamedRerouteDeclaration(None=MaterialFunctionCall(UVs=@n5, Offset=Constant(), Linear Mask=NamedRerouteUsage(), Non Linear Mask=NamedRerouteUsage(), Tangent Multiplier=Constant() {r=1}) {materialFunction=&MF_SmokeTrail_Waving}.Result) {name=Waving WPO, nodeColor=<rgba 1 0 0.7969 1>}
  n26 = Multiply(A=Subtract(A=@n24, B=Multiply(A=@n20, B=@n23))) {constB=0.01}
  n27 = VertexInterpolator(VS=Divide(
    A=Transform(None=@n26) {transformSourceType=TRANSFORMSOURCE_Local}
    B=MaterialFunctionCall() {materialFunction=&ObjectScale}.Scale XYZ
   ))
  NamedRerouteDeclaration(None=@n27.PS) {name=Normals, nodeColor=<rgba 0.4062 0 1 1>}
  NamedRerouteUsage()
  Reroute(None=@n27.PS)
  Saturate(None=@n1)
  TextureCoordinate()
  VertexInterpolator(VS=NamedRerouteUsage())
 ]
}
